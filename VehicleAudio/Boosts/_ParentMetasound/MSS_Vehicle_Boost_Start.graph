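# Parent stereo MetaSound one-shot for a vehicle's boost-start sound, with array, volume, pitch and timing inputs.
# On play it fires a shuffled main sample, optionally pitch-shifted by eased normal speed, plus a delayed, repeatable one-shot layer alternating between two shuffled players.
# Both layers mix with separate decibel gains.
# Trigger accumulation reports finished once both layers complete, with empty-array checks firing so a missing layer does not block it.
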MetaSoundSource /VehicleAudio/Boosts/_ParentMetasound/MSS_Vehicle_Boost_Start
inputs [
 Boost.Pitch.Max {type=Float, access=Constructor}
 Boost.Pitch.Min {type=Float, access=Constructor}
 Boost.PitchByNormalSpeed {type=Bool, access=Constructor}
 Boost.Start.Oneshots.DelayTime {type=Time, access=Constructor}
 Boost.Start.Oneshots.NumRepeats {type=Int32, access=Constructor}
 Boost.Start.Oneshots.RepeatSpeed {type=Time, access=Constructor}
 Boost.Start.Oneshots.Volume {type=Float, access=Constructor}
 Boost.Start.Oneshots.WaveAsset {type=WaveAsset:Array, access=Constructor}
 Boost.Start.Volume {type=Float, access=Constructor}
 Boost.Start.WaveAsset {type=WaveAsset:Array, access=Constructor}
 Fort.Vehicle.Acceleration {type=Float}
 Fort.Vehicle.BlowoffStrength {type=Float}
 Fort.Vehicle.BonusSpeed {type=Float}
 Fort.Vehicle.EnableShifting {type=Bool}
 Fort.Vehicle.Gear {type=Int32}
 Fort.Vehicle.IsGrounded {type=Bool}
 Fort.Vehicle.Load {type=Float}
 Fort.Vehicle.NormalSpeed {type=Float}
 Fort.Vehicle.OnBlowoff {type=Trigger}
 Fort.Vehicle.OnDownshift {type=Trigger}
 Fort.Vehicle.OnUpshift {type=Trigger}
 Fort.Vehicle.Pitch {type=Float}
 Fort.Vehicle.RPM {type=Float}
 Fort.Vehicle.Throttle {type=Float}
 UE.Source.OnPlay {type=Trigger}
 UE.Source.Orientation.Azimuth {type=Float}
 UE.Source.Orientation.Elevation {type=Float}
]
interfaces [Fort.Vehicle:1.0, UE.OutputFormat.Stereo:1.0, UE.Source.OneShot:1.0, UE.Source.Orientation:1.0, UE.Source:1.1]
outputs [
 "UE.OutputFormat.Stereo.Audio:0" {type=Audio}
 "UE.OutputFormat.Stereo.Audio:1" {type=Audio}
 UE.Source.OneShot.OnFinished {type=Trigger}
]
pages [Default {
  input_defaults {
   Boost.Pitch.Max=2.5
   Boost.Pitch.Min=0
   Boost.PitchByNormalSpeed="<MetaSoundBuilderResult.SUCCEEDED: 0>"
   Boost.Start.Oneshots.DelayTime=0.54
   Boost.Start.Oneshots.NumRepeats=1
   Boost.Start.Oneshots.RepeatSpeed=0.2
   Boost.Start.Oneshots.Volume=0
   Boost.Start.Volume=0
   Fort.Vehicle.Acceleration=0
   Fort.Vehicle.BlowoffStrength=0
   Fort.Vehicle.BonusSpeed=0
   Fort.Vehicle.EnableShifting=null
   Fort.Vehicle.Gear=0
   Fort.Vehicle.IsGrounded="<MetaSoundBuilderResult.SUCCEEDED: 0>"
   Fort.Vehicle.Load=0
   Fort.Vehicle.NormalSpeed=0.62
   Fort.Vehicle.OnBlowoff=null
   Fort.Vehicle.OnDownshift=null
   Fort.Vehicle.OnUpshift=null
   Fort.Vehicle.Pitch=0
   Fort.Vehicle.RPM=0
   Fort.Vehicle.Throttle=0
   UE.Source.OnPlay=null
   UE.Source.Orientation.Azimuth=0
   UE.Source.Orientation.Elevation=0
   Boost.Start.Oneshots.WaveAsset []
   Boost.Start.WaveAsset []
  }
  nodes [
   n1 = TriggerCompare.Clamp:Int32(Compare=^UE.Source.OnPlay, A=Array.Num:WaveAsset:Array(Array=^Boost.Start.Oneshots.WaveAsset).Num)
   TriggerDelay_0 = UE.Trigger Delay(Delay Time=^Boost.Start.Oneshots.DelayTime, In=@n1.False)
   TriggerSequence2_0 = "UE.Trigger Sequence (2)"(In=UE.TriggerRepeat(
     Period=^Boost.Start.Oneshots.RepeatSpeed
     Num Repeats=Min.Min:Int32(A=^Boost.Start.Oneshots.NumRepeats) {also {defaults {B=1}}}.Value
     Start=@TriggerDelay_0.Out
    ).RepeatOut)
   n2 = Array.Shuffle:WaveAsset:Array(In Array=^Boost.Start.Oneshots.WaveAsset, Next=@TriggerSequence2_0.Out 1)
   WavePlayer_0 = UE.Wave Player:Stereo(Play=@n2.On Next, Wave Asset=@n2.Value) {also {version [1,1]}}
   n3 = Array.Shuffle:WaveAsset:Array(In Array=^Boost.Start.Oneshots.WaveAsset, Next=@TriggerSequence2_0.Out 0)
   WavePlayer_1 = UE.Wave Player:Stereo(Play=@n3.On Next, Wave Asset=@n3.Value) {also {version [1,1]}}
   AudioMixerAudioMixerStereo2_0 = "AudioMixer.Audio Mixer (Stereo, 2)"(In 1 L=@WavePlayer_0.Out Left, In 1 R=@WavePlayer_0.Out Right, In 0 L=@WavePlayer_1.Out Left, In 0 R=@WavePlayer_1.Out Right)
   DecibelstoLinearGain_0 = UE.Decibels to Linear Gain:Float(Decibels=^Boost.Start.Volume)
   DecibelstoLinearGain_1 = UE.Decibels to Linear Gain:Float(Decibels=^Boost.Start.Oneshots.Volume)
   n4 = TriggerCompare.Clamp:Int32(Compare=^UE.Source.OnPlay, A=Array.Num:WaveAsset:Array(Array=^Boost.Start.WaveAsset).Num) {also {defaults {B=0, Type=3}}}
   n5 = Array.Shuffle:WaveAsset:Array(In Array=UE.Reroute(Value=^Boost.Start.WaveAsset).Value, Next=@n4.True)
   n6 = InitVariable.Float()
   n7 = TriggerCompare.Clamp:Bool(A=^Boost.PitchByNormalSpeed, Compare=^UE.Source.OnPlay) {also {defaults {B="<MetaSoundBuilderResult.SUCCEEDED: 0>"}}}
   n8 = UE.Multiply:Float(
    AdditionalOperands=MapRange.MapRange:Float(Out Range B=^Boost.Pitch.Max, Out Range A=^Boost.Pitch.Min, In=/VehicleAudio/Patches/EaseInQuart(Input=^Fort.Vehicle.NormalSpeed).Output).Out Value
    PrimaryOperand=Value.Value:Float(Reset=@n7.False, Set=@n7.True) {also {defaults {Init Value=0, Target Value=1}}}.Output Value
   )
   WavePlayer_2 = UE.Wave Player:Stereo(
    Play=@n5.On Next
    Wave Asset=@n5.Value
    Pitch Shift=VariableAccessor.Float(Variable=VariableMutator.Float(Variable=@n6.Variable, Value=@n8.Out).Variable).Value
   ) {also {version [1,1]}}
   AudioMixerAudioMixerStereo2_1 = "AudioMixer.Audio Mixer (Stereo, 2)"(
    In 1 L=@AudioMixerAudioMixerStereo2_0.Out L
    In 1 R=@AudioMixerAudioMixerStereo2_0.Out R
    Gain 0=@DecibelstoLinearGain_0.Linear Gain
    Gain 1=@DecibelstoLinearGain_1.Linear Gain
    In 0 L=@WavePlayer_2.Out Left
    In 0 R=@WavePlayer_2.Out Right
   )
   TriggerAnyTriggerAny2_0 = "TriggerAny.Trigger Any (2)"(In 1=UE.Reroute(Value=UE.Reroute(Value=@n4.False).Value).Value, In 0=@WavePlayer_2.On Finished)
   TriggerAnyTriggerAny3_0 = "TriggerAny.Trigger Any (3)"(In 0=@n1.True, In 2=@WavePlayer_0.On Finished, In 1=@WavePlayer_1.On Finished)
   TriggerAccumulatorTriggerAccumulate2_0 = "TriggerAccumulator.Trigger Accumulate (2)"(In 0=@TriggerAnyTriggerAny2_0.Out, In 1=@TriggerAnyTriggerAny3_0.Out)
  ]
  links [
   {dst="@UE.OutputFormat.Stereo.Audio:0", src=AudioMixerAudioMixerStereo2_1.Out L}
   {dst="@UE.OutputFormat.Stereo.Audio:1", src=AudioMixerAudioMixerStereo2_1.Out R}
   {dst="@UE.Source.OneShot.OnFinished", src=TriggerAccumulatorTriggerAccumulate2_0.Out}
  ]
 }]
properties {
 assetImportData=&self:AssetImportData
 attenuationSettings=/VehicleAudio/Boosts/Vehicle_Boost_Att
 audioPropertiesSheet=/VehicleAudio/PropertySheets/APS_Cosmetic_Vehicle_Boost_Start
 maxDistance=5012
 numChannels=2
 outputFormat=Stereo
 sampleRate=48000
 soundClassObject=/Game/Sounds/FortSoundClassesAndModes/Fort_Vehicle_Exterior_Boost
 soundSubmixObject=&Fortnite_SFX
 concurrencySet [/VehicleAudio/Boosts/Vehicle_Boost_Start_Conc, /VehicleAudio/Boosts/Vehicle_Boost_End_Conc]
 modulationSettings {
  highpassRouting=Union
  lowpassRouting=Union
  volumeRouting=Union
  highpassModulationDestination {bEnableModulation=true, modulators [/Game/Sounds/Modulation/Patches/HPF/PP_FortVehicles_Boost_HPF]}
  lowpassModulationDestination {bEnableModulation=true, modulators [/Game/Sounds/Modulation/Patches/LPF/PP_FortVehicles_Boost_LPF]}
  volumeModulationDestination {
   bEnableModulation=true
   modulators [/VehicleAudio/Modulation/DM_VehicleTurboPatch, /Game/Sounds/Modulation/Patches/PP_FortVehicles_Boost_Vol]
  }
 }
}
